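# One-shot mono MetaSound source that plays a random rubber ducky spawn sound for the FunkyFeather emote.
# On play, it picks one of eleven waves and feeds it to a wave player with pitch shift randomized around a base.
# A decibel volume, randomizable but with zero default spread, scales the audio before a mixer gain boost.
# It uses duck-specific attenuation and an owner-limited concurrency override with a retrigger time.
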MetaSoundSource /BRCosmetics/Sounds/Emotes/FunkyFeather/MSS_Emote_FunkyFeather_RubberDucky_OneShotArrayRandomized
inputs [
 Random.Pitch.Base {type=Float}
 Random.Pitch.Spread {type=Float}
 Random.Volume.Base {type=Float}
 Random.Volume.Spread {type=Float}
 Sounds {type=WaveAsset:Array}
 UE.Source.OnPlay {type=Trigger}
]
interfaces [UE.OutputFormat.Mono:1.0, UE.Source.OneShot:1.0, UE.Source:1.1]
outputs [
 "UE.OutputFormat.Mono.Audio:0" {type=Audio}
 UE.Source.OneShot.OnFinished {type=Trigger}
]
pages [Default {
  input_defaults {
   Random.Pitch.Base=0
   Random.Pitch.Spread=1
   Random.Volume.Base=1
   Random.Volume.Spread=0
   UE.Source.OnPlay=null
   Sounds [
    &/Emote_FunkyFeather_RubberDucky_Spawn_01
    &/Emote_FunkyFeather_RubberDucky_Spawn_02
    &/Emote_FunkyFeather_RubberDucky_Spawn_03
    &/Emote_FunkyFeather_RubberDucky_Spawn_04
    &/Emote_FunkyFeather_RubberDucky_Spawn_05
    &/Emote_FunkyFeather_RubberDucky_Spawn_06
    &/Emote_FunkyFeather_RubberDucky_Spawn_07
    &/Emote_FunkyFeather_RubberDucky_Spawn_08
    &/Emote_FunkyFeather_RubberDucky_Spawn_09
    &/Emote_FunkyFeather_RubberDucky_Spawn_10
    &/Emote_FunkyFeather_RubberDucky_Spawn_11
   ]
  }
  nodes [
   n1 = AbsoluteValue.Abs:Float(Input=^Random.Volume.Spread)
   DecibelstoLinearGain_0 = UE.Decibels to Linear Gain:Float(Decibels=UE.RandomFloat(
     Next=^UE.Source.OnPlay
     Max=UE.Add:Float(PrimaryOperand=^Random.Volume.Base, AdditionalOperands=@n1.Value).Out
     Min=UE.Subtract:Float(PrimaryOperand=^Random.Volume.Base, AdditionalOperands=@n1.Value).Out
    ).Value)
   ArrayRandomGet_0 = Array.Random Get:WaveAsset:Array(In Array=^Sounds, Next=^UE.Source.OnPlay) {also {defaults {No Repeats=-1, Weights []}}}
   n2 = AbsoluteValue.Abs:Float(Input=^Random.Pitch.Spread)
   WavePlayer_0 = UE.Wave Player:Mono(
    Play=@ArrayRandomGet_0.On Next
    Wave Asset=@ArrayRandomGet_0.Value
    Pitch Shift=UE.RandomFloat(
     Next=^UE.Source.OnPlay
     Max=UE.Add:Float(PrimaryOperand=^Random.Pitch.Base, AdditionalOperands=@n2.Value).Out
     Min=UE.Subtract:Float(PrimaryOperand=^Random.Pitch.Base, AdditionalOperands=@n2.Value).Out
    ).Value
   ) {also {version [1,1]}}
   AudioMixerAudioMixerMono2_0 = "AudioMixer.Audio Mixer (Mono, 2)"(In 0=UE.Multiply:Audio by Float(AdditionalOperands=@DecibelstoLinearGain_0.Linear Gain, PrimaryOperand=@WavePlayer_0.Out Mono).Out) {also {defaults {Gain 0=1.6}}}
  ]
  links [
   {dst="@UE.OutputFormat.Mono.Audio:0", src=AudioMixerAudioMixerMono2_0.Out}
   {dst="@UE.Source.OneShot.OnFinished", src=WavePlayer_0.On Finished}
  ]
 }]
properties {
 assetImportData=&self:AssetImportData
 attenuationSettings=/Game/Sounds/Attenuation/Fort_Emotes_FunkyFeather_Ducks_Att
 audioPropertiesSheet=/BRCosmetics/Sounds/Settings/PropertySheets/Emotes/APS_Cosmetic_Emote_Foley
 bOverrideConcurrency=true
 maxDistance=800
 sampleRate=48000
 soundClassObject=/Game/Sounds/FortSoundClassesAndModes/Fort_Foley_Emote
 concurrencyOverrides {bLimitToOwner=true, maxCount=2, retriggerTime=0.8}
 soundSubmixSends [{}]
}
